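# Physics asset for the Uproar Braids backpack cosmetic, simulating a figure with cymbal arms, head, jaw, necklace and crank.
# Capsule and box bodies form constrained chains branching from dyn_base to both arms, the necklace, the neck-head-jaw and the crank.
# Arm, necklace and crank chains start at MSTR bodies with free angular joints, followed by tighter limits.
# Arms lock angularly to clavicles, jaw segments swing mainly on one axis, and the crank only twists.
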PhysicsAsset /BRCosmetics/Accessories/FORT_Backpacks/Backpack_F_MED_Uproar_Braids/Meshes/F_MED_Uproar_Braids_Pack_Arms_Physics
bodies {
 dyn_L_arm [{shapeType=Capsule, length=10.6407, radius=3, shapeName=dyn_L_arm_capsule, center=<6.1646 0 -0.0001>, rotation=<rot 90 0 0>}]
 dyn_L_clav [{
   shapeType=Capsule
   length=1.6845
   radius=1.3917
   shapeName=dyn_L_clav_capsule
   center=<2.0038 -0.0972 0.0524>
   rotation=<rot 0 93.3513 -88.5963>
  }]
 dyn_L_clav_MSTR [{shapeType=Capsule, length=0.505, radius=0.505, shapeName=dyn_L_clav_MSTR_capsule, rotation=<rot 0 -90 0>}]
 dyn_L_cymbal [{shapeType=Box, extentX=14.3533, extentY=5.4703, extentZ=14.6112, shapeName=dyn_L_cymbal_box, center=<1 0 0>, rotation=<rot 0 0 90>}]
 dyn_R_arm [{shapeType=Capsule, length=10.6653, radius=3, shapeName=dyn_R_arm_capsule, center=<6.1646 0 -0.0001>, rotation=<rot 90 0 0>}]
 dyn_R_clav [{shapeType=Capsule, length=0.505, radius=0.505, shapeName=dyn_R_clav_capsule, rotation=<rot 0 -90 0>}]
 dyn_R_clav_MSTR [{shapeType=Capsule, length=0.505, radius=0.505, shapeName=dyn_R_clav_MSTR_capsule, rotation=<rot 0 -90 0>}]
 dyn_R_cymbal [{shapeType=Box, extentX=14.1304, extentY=5.441, extentZ=14.2312, shapeName=dyn_R_cymbal_box, center=<1 0 0>, rotation=<rot 0 0 90>}]
 dyn_base [{shapeType=Capsule, length=6.8927, radius=6.7594, center=<-2.9731 -7.9347 -0.1451>, rotation=<rot 0 107.0991 -89.4021>}]
 dyn_crank [{shapeType=Capsule, length=5.2755, radius=3.6709, shapeName=dyn_crank_capsule, center=<0 -5.091 -0.0001>, rotation=<rot 0 0 90>}]
 dyn_crank_MSTR [{shapeType=Capsule, length=0.505, radius=0.505, shapeName=dyn_crank_MSTR_capsule, rotation=<rot 0 -90 0>}]
 dyn_head [{shapeType=Capsule, length=3.3474, radius=4.2609, shapeName=dyn_head_capsule, center=<-0.0907 1.8065 0.1981>, rotation=<rot 0.3204 -90 -0.65>}]
 dyn_jaw1 [{
   shapeType=Box
   extentX=1.25
   extentY=2.6916
   extentZ=5.9707
   shapeName=dyn_jaw1_box
   center=<0.716 -0.5292 0.0126>
   rotation=<rot -0.4865 -114.9941 -1.1219>
  }]
 dyn_jaw2 [{
   shapeType=Box
   extentX=0.7166
   extentY=2.0883
   extentZ=4.6371
   shapeName=dyn_jaw2_box
   center=<0.695 -0.0754 -0.0852>
   rotation=<rot 1.3156 -90 7.1305>
  }]
 dyn_jaw3 [{
   shapeType=Box
   extentX=4.5888
   extentY=2
   extentZ=6.9319
   shapeName=dyn_jaw3_box
   center=<0.7673 1.6331 -0.0228>
   rotation=<rot -0.5133 -124.9603 -2.7454>
  }]
 dyn_neck [{shapeType=Capsule, length=4.3451, radius=1.772, shapeName=dyn_neck_capsule, center=<3.6161 0 0.0001>, rotation=<rot 0 90 90>}]
 dyn_necklace1 [{
   shapeType=Capsule
   length=4.0064
   radius=1.3415
   shapeName=dyn_necklace1_capsule
   center=<7.3149 0.0001 -0.0375>
   rotation=<rot 1.3117 -90 1.7544>
  }]
 dyn_necklace2 [{shapeType=Capsule, length=0.505, radius=0.505, shapeName=dyn_necklace2_capsule, rotation=<rot 0 -90 0>}]
 dyn_necklace3 [{shapeType=Capsule, length=3.4027, radius=3.2562, shapeName=dyn_necklace3_capsule, center=<4.0631 0 0>, rotation=<rot 90 0 0>}]
 dyn_necklace_MSTR [{shapeType=Capsule, length=0.505, radius=0.505, shapeName=dyn_necklace_MSTR_capsule, rotation=<rot 0 -90 0>}]
}
constraints [
 {
  bone1Name=dyn_R_clav_MSTR
  bone2Name=dyn_base
  swing1LimitDegrees=45
  swing1Motion=Free
  swing2LimitDegrees=45
  swing2Motion=Free
  twistLimitDegrees=45
  twistMotion=Free
 }
 {bone1Name=dyn_R_clav, bone2Name=dyn_R_clav_MSTR, swing1LimitDegrees=15, swing2LimitDegrees=25, twistLimitDegrees=15}
 {
  bone1Name=dyn_R_arm
  bone2Name=dyn_R_clav
  swing1LimitDegrees=15
  swing1Motion=Locked
  swing2LimitDegrees=15
  swing2Motion=Locked
  twistLimitDegrees=2
  twistMotion=Locked
 }
 {bone1Name=dyn_R_cymbal, bone2Name=dyn_R_arm, swing1LimitDegrees=25, swing2LimitDegrees=25, twistLimitDegrees=2}
 {
  bone1Name=dyn_L_clav_MSTR
  bone2Name=dyn_base
  swing1LimitDegrees=45
  swing1Motion=Free
  swing2LimitDegrees=45
  swing2Motion=Free
  twistLimitDegrees=45
  twistMotion=Free
 }
 {bone1Name=dyn_L_clav, bone2Name=dyn_L_clav_MSTR, swing1LimitDegrees=15, swing2LimitDegrees=25, twistLimitDegrees=15}
 {
  bone1Name=dyn_L_arm
  bone2Name=dyn_L_clav
  swing1LimitDegrees=15
  swing1Motion=Locked
  swing2LimitDegrees=15
  swing2Motion=Locked
  twistLimitDegrees=2
  twistMotion=Locked
 }
 {bone1Name=dyn_L_cymbal, bone2Name=dyn_L_arm, swing1LimitDegrees=25, swing2LimitDegrees=25, twistLimitDegrees=2}
 {
  bone1Name=dyn_necklace_MSTR
  bone2Name=dyn_base
  swing1LimitDegrees=45
  swing1Motion=Free
  swing2LimitDegrees=45
  swing2Motion=Free
  twistLimitDegrees=45
  twistMotion=Free
 }
 {bone1Name=dyn_necklace1, bone2Name=dyn_necklace_MSTR, swing1LimitDegrees=10, swing2LimitDegrees=10, twistLimitDegrees=2}
 {bone1Name=dyn_necklace2, bone2Name=dyn_necklace1, swing1LimitDegrees=10, swing2LimitDegrees=10, twistLimitDegrees=2}
 {bone1Name=dyn_necklace3, bone2Name=dyn_necklace2, swing1LimitDegrees=20, swing2LimitDegrees=20, twistLimitDegrees=2}
 {bone1Name=dyn_neck, bone2Name=dyn_base, swing1LimitDegrees=20, swing2LimitDegrees=15, twistLimitDegrees=2}
 {bone1Name=dyn_head, bone2Name=dyn_neck, swing1LimitDegrees=10, swing2LimitDegrees=15, twistLimitDegrees=5}
 {bone1Name=dyn_jaw1, bone2Name=dyn_head, swing1LimitDegrees=10, swing2LimitDegrees=0, twistLimitDegrees=2}
 {bone1Name=dyn_jaw2, bone2Name=dyn_jaw1, swing1LimitDegrees=10, swing2LimitDegrees=0, twistLimitDegrees=2}
 {bone1Name=dyn_jaw3, bone2Name=dyn_jaw2, swing1LimitDegrees=10, swing2LimitDegrees=0, twistLimitDegrees=2}
 {
  bone1Name=dyn_crank_MSTR
  bone2Name=dyn_base
  swing1LimitDegrees=45
  swing1Motion=Free
  swing2LimitDegrees=45
  swing2Motion=Free
  twistLimitDegrees=45
  twistMotion=Free
 }
 {bone1Name=dyn_crank, bone2Name=dyn_crank_MSTR, swing1LimitDegrees=0, swing2LimitDegrees=0, twistLimitDegrees=45}
]
