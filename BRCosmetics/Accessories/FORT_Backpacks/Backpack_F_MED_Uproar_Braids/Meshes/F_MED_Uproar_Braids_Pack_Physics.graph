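# Physics asset for the Uproar Braids backpack cosmetic, defining collision capsules and joints for a base and two dangling leg chains.
# Each left and right chain runs from a tiny master body on the base through a leg capsule to a foot capsule.
# The master joints rotate freely on all angular axes, while leg and foot joints have tight limits that favor one swing axis, so each chain bends mostly in one plane.
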PhysicsAsset /BRCosmetics/Accessories/FORT_Backpacks/Backpack_F_MED_Uproar_Braids/Meshes/F_MED_Uproar_Braids_Pack_Physics
bodies {
 dyn_L_foot [{shapeType=Capsule, length=5.5, radius=2.25, shapeName=dyn_L_foot_capsule, center=<4 0 0>, rotation=<rot 90 0 0>}]
 dyn_L_leg [{shapeType=Capsule, length=4, radius=3, shapeName=dyn_L_leg_capsule, center=<3 0 0>, rotation=<rot 90 0 0>}]
 dyn_L_leg_MSTR [{shapeType=Capsule, length=0.505, radius=0.505, shapeName=dyn_L_leg_MSTR_capsule, rotation=<rot 0 -90 0>}]
 dyn_R_foot [{shapeType=Capsule, length=5.5, radius=2.25, shapeName=dyn_R_foot_capsule, center=<4 0 0>, rotation=<rot 90 0 0>}]
 dyn_R_leg [{shapeType=Capsule, length=4, radius=3, shapeName=dyn_R_leg_capsule, center=<3 0 0>, rotation=<rot 90 0 0>}]
 dyn_R_leg_MSTR [{shapeType=Capsule, length=0.505, radius=0.505, shapeName=dyn_R_leg_MSTR_capsule, rotation=<rot 0 -90 0>}]
 dyn_base [{shapeType=Capsule, length=5.8927, radius=7.7594, center=<-2.9731 -7.9347 -0.1451>, rotation=<rot 0 107.0991 -89.4021>}]
}
constraints [
 {
  bone1Name=dyn_L_leg_MSTR
  bone2Name=dyn_base
  swing1LimitDegrees=45
  swing1Motion=Free
  swing2LimitDegrees=45
  swing2Motion=Free
  twistLimitDegrees=45
  twistMotion=Free
 }
 {
  bone1Name=dyn_R_leg_MSTR
  bone2Name=dyn_base
  swing1LimitDegrees=45
  swing1Motion=Free
  swing2LimitDegrees=45
  swing2Motion=Free
  twistLimitDegrees=45
  twistMotion=Free
 }
 {bone1Name=dyn_L_leg, bone2Name=dyn_L_leg_MSTR, swing1LimitDegrees=15, swing2LimitDegrees=2, twistLimitDegrees=2}
 {bone1Name=dyn_R_leg, bone2Name=dyn_R_leg_MSTR, swing1LimitDegrees=15, swing2LimitDegrees=2, twistLimitDegrees=2}
 {bone1Name=dyn_R_foot, bone2Name=dyn_R_leg, swing1LimitDegrees=15, swing2LimitDegrees=2, twistLimitDegrees=2}
 {bone1Name=dyn_L_foot, bone2Name=dyn_L_leg, swing1LimitDegrees=15, swing2LimitDegrees=2, twistLimitDegrees=2}
]
