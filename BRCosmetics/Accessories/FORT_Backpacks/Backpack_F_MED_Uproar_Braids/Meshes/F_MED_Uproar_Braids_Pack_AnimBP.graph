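# Animation Blueprint for the Uproar Braids back bling: a cranked figure that claps, laughs and tilts its head when the owner's kill count rises.
# Two chained RigidBody sims supply secondary motion, with AnimDynamics replacing them in the frontend or when rigid bodies are disabled.
# Additive bone edits layer sine-driven arm, head and jaw motion, fading over a tick-counted timer.
# SpinFunction continuously spins the crank, overwriting the damage-based crank rotation each update.
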AnimBlueprint /BRCosmetics/Accessories/FORT_Backpacks/Backpack_F_MED_Uproar_Braids/Meshes/F_MED_Uproar_Braids_Pack_AnimBP
target_skeleton=&/F_MED_Uproar_Braids_Pack_Skeleton
graphs [
 AnimGraph {nodes [
   Component To Local()
   Local To Component()
   n1 = Use cached pose 'Sim Cache'()
   n2 = Use cached pose 'NoRBAN Cache'()
   n3 = Get NoRBAN or FE Alpha() -> (NoRBAN or FE Alpha:bool)
   n4 = "Transform (Modify) Bone - Bone: dyn_crank"(
    ComponentPose=Local To Component(LocalPose=Blend Poses by bool(BlendPose_0=@n2, BlendPose_1=@n1, bActiveValue=@n3) {blendType=Linear})
    Alpha=1
    Translation="0,0,0"
    Rotation=Get CrankRotator() -> (CrankRotator:Rotator)
    Scale="1,1,1"
   ) {rotationMode=BMM_Additive, rotationSpace=BCS_BoneSpace, alphaBoolBlend {blendOption=Linear}, boneToModify {boneName=dyn_crank}}
   n5 = Get Clap Alpha() -> (Clap Alpha:double)
   n6 = "Transform (Modify) Bone - Bone: dyn_L_clav"(ComponentPose=@n4, Alpha=@n5, Translation="0,0,0", Rotation=Get L Arm Rot() -> (L Arm Rot:Rotator), Scale="1,1,1") {rotationMode=BMM_Additive, rotationSpace=BCS_BoneSpace, alphaBoolBlend {blendOption=Linear}, boneToModify {boneName=dyn_L_clav}}
   n7 = "Transform (Modify) Bone - Bone: dyn_R_clav"(ComponentPose=@n6, Alpha=@n5, Translation="0,0,0", Rotation=Get R Arm Rot() -> (R Arm Rot:Rotator), Scale="1,1,1") {rotationMode=BMM_Additive, rotationSpace=BCS_BoneSpace, alphaBoolBlend {blendOption=Linear}, boneToModify {boneName=dyn_R_clav}}
   n8 = "Transform (Modify) Bone - Bone: dyn_head"(ComponentPose=@n7, Alpha=@n5, Translation="0,0,0", Rotation=Get Head Laugh Rot() -> (Head Laugh Rot:Rotator), Scale="1,1,1") {
    rotationMode=BMM_Additive
    rotationSpace=BCS_BoneSpace
    alphaBoolBlend {blendOption=Linear, blendOutTime=0.2}
    boneToModify {boneName=dyn_head}
    rotation=<rot -20 10 -35>
   }
   n9 = Get Jaw Laugh Rot() -> (Jaw Laugh Rot:Rotator)
   n10 = "Transform (Modify) Bone - Bone: dyn_jaw1"(ComponentPose=@n8, Alpha=@n5, Translation="0,0,0", Rotation=@n9, Scale="1,1,1") {
    rotationMode=BMM_Additive
    rotationSpace=BCS_BoneSpace
    alphaBoolBlend {blendOption=Linear, blendOutTime=0.2}
    boneToModify {boneName=dyn_jaw1}
    rotation=<rot -20 10 -35>
   }
   n11 = "Transform (Modify) Bone - Bone: dyn_jaw2"(ComponentPose=@n10, Alpha=@n5, Translation="0,0,0", Rotation=@n9, Scale="1,1,1") {
    rotationMode=BMM_Additive
    rotationSpace=BCS_BoneSpace
    alphaBoolBlend {blendOption=Linear, blendOutTime=0.2}
    boneToModify {boneName=dyn_jaw2}
    rotation=<rot -20 10 -35>
   }
   n12 = Get Active_Kill() -> (Active_Kill:bool)
   n13 = "Transform (Modify) Bone - Bone: dyn_neck_MSTR"!(ComponentPose=@n11, bAlphaBoolEnabled:bool=@n12, Translation="0,0,0", Rotation="0.000,-20,35", Scale="1,1,1") -> (Pose) {
    alphaInputType=Bool
    rotationMode=BMM_Additive
    rotationSpace=BCS_BoneSpace
    alphaBoolBlend {blendInTime=0.2, blendOption=Linear, blendOutTime=0.2}
    boneToModify {boneName=dyn_neck_MSTR}
    rotation {roll=35, yaw=-20}
   }
   n14 = "Transform (Modify) Bone - Bone: dyn_head"!(ComponentPose=@n13, bAlphaBoolEnabled:bool=@n12, Translation="0,0,0", Rotation="-15,10,-35", Scale="1,1,1") -> (Pose) {
    alphaInputType=Bool
    rotationMode=BMM_Additive
    rotationSpace=BCS_BoneSpace
    alphaBoolBlend {blendInTime=0.2, blendOption=Linear, blendOutTime=0.2}
    boneToModify {boneName=dyn_head}
    rotation=<rot -15 10 -35>
   }
   Output Pose(Result=Component To Local(ComponentPose=@n14)) {layerGroup=None}
   Save cached pose 'PreSim'(Pose=Local Space Ref Pose()) {cacheName=PreSim}
   n15 = RigidBody!(
    ComponentPose=Local To Component(LocalPose=Use cached pose 'PreSim'())
    bAlphaBoolEnabled:bool=NOT Boolean(A=Get Active_Kill())
    ExternalForce
   ) -> (Pose) {
    alphaInputType=Bool
    bClampLinearTranslationLimitToRefPose=true
    bOverrideWorldGravity=true
    overridePhysicsAsset=&/F_MED_Uproar_Braids_Pack_Arms_Physics
    simulationSpace=BaseBoneSpace
    alphaBoolBlend {blendOption=Linear, blendOutTime=0.2}
    baseBoneRef {boneName=root}
    componentAppliedLinearAccClamp=<3000 3000 3000>
    componentLinearAccScale=<0.3 0.3 0.35>
    componentLinearVelScale=<0.1 0.1 0.1>
    overrideWorldGravity {z=-980}
   }
   n16 = RigidBody!(ComponentPose=@n15, Alpha:float, ExternalForce) -> (Pose) {
    bClampLinearTranslationLimitToRefPose=true
    bOverrideWorldGravity=true
    overridePhysicsAsset=&/F_MED_Uproar_Braids_Pack_Physics
    simulationSpace=BaseBoneSpace
    alphaBoolBlend {blendOption=Linear}
    baseBoneRef {boneName=root}
    componentAppliedLinearAccClamp=<3000 3000 3000>
    componentLinearAccScale=<0.25 0.25 0.25>
    componentLinearVelScale=<0.1 0.1 0.1>
    overrideWorldGravity {z=-980}
   }
   Save cached pose 'Sim Cache'(Pose=Component To Local(ComponentPose=@n16)) {cacheName=Sim Cache}
   n17 = Anim Dynamics - Bone: dyn_neck(ComponentPose=Local To Component(LocalPose=Use cached pose 'PreSim'()), Alpha=1, ExternalForce="0,0,0") {
    angularDampingOverride=0.995
    angularSpringConstant=5000
    bAngularSpring=true
    bOverrideAngularDamping=true
    bUseGravityOverride=true
    bUsePlanarLimit=false
    alphaBoolBlend {blendOption=Linear}
    boundBone {boneName=dyn_neck}
    componentAppliedLinearAccClamp=<3000 3000 3000>
    componentLinearAccScale=<0.25 0.25 0.35>
    componentLinearVelScale=<0.2 0.2 0.3>
    gravityOverride {z=-800}
    physicsBodyDefinitions [{
      boundBone {boneName=dyn_neck}
      boxExtents=<15 25 25>
      constraintSetup {angularLimitsMax=<1 5 10>, angularLimitsMin=<-1 -5 -25>, angularTarget=<1 0 0>}
      localJointOffset=<2 0 0>
     }]
   }
   n18 = Anim Dynamics - Bone: dyn_head(ComponentPose=@n17, Alpha=1, ExternalForce="0,0,0") {
    angularDampingOverride=0.995
    angularSpringConstant=10000
    bAngularSpring=true
    bOverrideAngularDamping=true
    bUseGravityOverride=true
    bUsePlanarLimit=false
    alphaBoolBlend {blendOption=Linear}
    boundBone {boneName=dyn_head}
    componentAppliedLinearAccClamp=<3000 3000 3000>
    componentLinearAccScale=<0.2 0.2 0.3>
    componentLinearVelScale=<0.2 0.2 0.3>
    gravityOverride {z=-800}
    physicsBodyDefinitions [{
      boundBone {boneName=dyn_head}
      boxExtents=<25 40 40>
      constraintSetup {angularLimitsMax=<1 5 10>, angularLimitsMin=<-1 -5 -25>, angularTarget=<1 0 0>}
      localJointOffset=<2 0 0>
     }]
   }
   n19 = Anim Dynamics - Bone: dyn_L_arm(ComponentPose=@n18, Alpha=1, ExternalForce="0,0,0") {
    angularDampingOverride=0.995
    angularSpringConstant=8000
    bAngularSpring=true
    bOverrideAngularDamping=true
    bUseGravityOverride=true
    bUsePlanarLimit=false
    alphaBoolBlend {blendOption=Linear}
    boundBone {boneName=dyn_L_arm}
    componentAppliedLinearAccClamp=<3000 3000 3000>
    componentLinearAccScale=<0.2 0.2 0.3>
    componentLinearVelScale=<0.2 0.2 0.3>
    gravityOverride {z=-800}
    physicsBodyDefinitions [{
      boundBone {boneName=dyn_L_arm}
      boxExtents=<32 20 32>
      constraintSetup {angularLimitsMax=<1 2.5 15>, angularLimitsMin=<-1 -12.5 -5>, angularTarget=<1 0 0>}
      localJointOffset=<2 0 0>
     }]
   }
   n20 = Anim Dynamics - Bone: dyn_R_arm(ComponentPose=@n19, Alpha=1, ExternalForce="0,0,0") {
    angularDampingOverride=0.995
    angularSpringConstant=8000
    bAngularSpring=true
    bOverrideAngularDamping=true
    bUseGravityOverride=true
    bUsePlanarLimit=false
    alphaBoolBlend {blendOption=Linear}
    boundBone {boneName=dyn_R_arm}
    componentAppliedLinearAccClamp=<3000 3000 3000>
    componentLinearAccScale=<0.2 0.2 0.3>
    componentLinearVelScale=<0.2 0.2 0.3>
    gravityOverride {z=-800}
    physicsBodyDefinitions [{
      boundBone {boneName=dyn_R_arm}
      boxExtents=<35 20 35>
      constraintSetup {angularLimitsMax=<1 12.5 15>, angularLimitsMin=<-1 -2.5 -10>, angularTarget=<1 0 0>}
      localJointOffset=<2 0 0>
     }]
   }
   n21 = Anim Dynamics - Bone: dyn_L_foot(ComponentPose=@n20, Alpha=1, ExternalForce="0,0,0") {
    angularDampingOverride=0.995
    angularSpringConstant=5000
    bAngularSpring=true
    bOverrideAngularDamping=true
    bUseGravityOverride=true
    bUsePlanarLimit=false
    alphaBoolBlend {blendOption=Linear}
    boundBone {boneName=dyn_L_foot}
    componentAppliedLinearAccClamp=<3000 3000 3000>
    componentLinearAccScale=<0.25 0.25 0.35>
    componentLinearVelScale=<0.2 0.2 0.3>
    gravityOverride {z=-800}
    physicsBodyDefinitions [{
      boundBone {boneName=dyn_L_foot}
      boxExtents=<32 20 32>
      constraintSetup {angularLimitsMax=<1 1 15>, angularLimitsMin=<-1 -1 -10>, angularTarget=<1 0 0>}
      localJointOffset=<2 0 0>
     }]
   }
   n22 = Anim Dynamics - Bone: dyn_R_foot(ComponentPose=@n21, Alpha=1, ExternalForce="0,0,0") {
    angularDampingOverride=0.995
    angularSpringConstant=5000
    bAngularSpring=true
    bOverrideAngularDamping=true
    bUseGravityOverride=true
    bUsePlanarLimit=false
    alphaBoolBlend {blendOption=Linear}
    boundBone {boneName=dyn_R_foot}
    componentAppliedLinearAccClamp=<3000 3000 3000>
    componentLinearAccScale=<0.25 0.25 0.35>
    componentLinearVelScale=<0.2 0.2 0.3>
    gravityOverride {z=-800}
    physicsBodyDefinitions [{
      boundBone {boneName=dyn_R_foot}
      boxExtents=<32 20 32>
      constraintSetup {angularLimitsMax=<1 1 15>, angularLimitsMin=<-1 -1 -10>, angularTarget=<1 0 0>}
      localJointOffset=<2 0 0>
     }]
   }
   Save cached pose 'NoRBAN Cache'(Pose=Component To Local(ComponentPose=@n22)) {cacheName=NoRBAN Cache}
  ]}
 EventGraph {nodes [
   Event BlueprintUpdateAnimation()
   > n1 = Sequence!(execute) -> (then_0, then_1, then_2:Exec, then_3:Exec)
   > Branch(Condition=Get Active_Kill())
   > Set Clap_Timer(Clap_Timer:double=float + float(A=Get Clap_Timer() -> (Clap_Timer:double), B=1) {is=title K2Node_CommutativeAssociativeBinaryOperator}) -> (Output_Get:double)
   n2 = float * float(A=GetGameTimeInSeconds(), B=750)
   n3 = "Absolute (Float)"(A=float * float(A="Sin (Degrees)"!(A:double=@n2) -> (ReturnValue:double), B=70))
   > Set L Arm Rot(L Arm Rot:Rotator=MakeRotator(Yaw=float * float(A=@n3, B=-1))) -> (Output_Get:Rotator)
   n4 = float * float(A=GetGameTimeInSeconds(), B=750)
   n5 = "Absolute (Float)"(A=float * float(A="Sin (Degrees)"(A=@n4), B=67))
   > Set R Arm Rot(R Arm Rot:Rotator=MakeRotator(Yaw=float * float(A=@n5, B=-1))) -> (Output_Get:Rotator)
   n6 = float * float(A=GetGameTimeInSeconds(), B=750)
   n7 = "Absolute (Float)"(A=float * float(A="Sin (Degrees)"(A=@n6), B=25))
   > Set Head Laugh Rot(Head Laugh Rot:Rotator=MakeRotator(Yaw=float * float(A=@n7, B=1))) -> (Output_Get:Rotator)
   n8 = float * float(A=GetGameTimeInSeconds(), B=750)
   n9 = "Absolute (Float)"(A=float * float(A="Sin (Degrees)"(A=@n8), B=35))
   > Set Jaw Laugh Rot(Jaw Laugh Rot:Rotator=MakeRotator(Yaw=float * float(A=@n9, B=-1))) -> (Output_Get:Rotator)
   n10 = "Clamp (Float)"(Value=Get Crank Count() -> (Crank Count:double), Min=-1080, Max=-180)
   n11 = float - float(A=Get Clap_Timer_Length() -> (Clap_Timer_Length:double), B=Get Clap_Timer()) {is=title K2Node_CallFunction}
   n12 = Get Clap_Timer_Length()
   > Set CrankRotator(CrankRotator:Rotator=MakeRotator(Roll=MapRangeClamped(Value=@n11, InRangeA=@n12, InRangeB=0, OutRangeA=@n10, OutRangeB=0))) -> (Output_Get:Rotator)
   n13 = Get Clap_Timer()
   n14 = Get Clap_Timer_Length()
   > Set Clap Alpha(Clap Alpha:double=MapRangeClamped(Value=float - float(A=@n14, B=@n13), InRangeA=Get Clap_Timer_Length(), InRangeB=0, OutRangeA=1, OutRangeB=0)) -> (Output_Get:double)
   > Branch(Condition="Equal (Float)"!(A:double=Get Clap Alpha(), B:double=0) -> (ReturnValue:bool) {is=title K2Node_CallFunction})
   > Set Crank Count(Crank Count:double=0) -> (Output_Get:double)
   > Set Active_Kill(Active_Kill:bool=false) -> (Output_Get:bool)
   > Set Active_Damage(Active_Damage:bool=false) -> (Output_Get:bool)
   > Set L Arm Rot(L Arm Rot="0, 0, 0")
   > Set R Arm Rot(R Arm Rot="0, 0, 0")
   > Set CrankRotator(CrankRotator="0, 0, 0")
   > Set Head Laugh Rot(Head Laugh Rot="0, 0, 0")
   > Set Jaw Laugh Rot(Jaw Laugh Rot="0, 0, 0")
   > Set Clap_Timer(Clap_Timer=0)
   > Delay(Duration=0.01)
   > Set Clap Alpha(Clap Alpha=1)
   Set deltaTimeRemap(
    execute=@n1.then_2
    deltaTimeRemap:double=FInterpTo(Current=Get deltaTimeRemap() -> (deltaTimeRemap:double), Target=1, DeltaTime=0, InterpSpeed=1)
   ) -> (Output_Get:double)
   n15 = SpinFunction!(execute=@n1.then_1, self, Spin:double=Get crankSpin() -> (crankSpin:double), Speed:double=1.25) -> (then, Output:double)
   > n16 = Set crankSpin(crankSpin:double=@n15) -> (Output_Get:double)
   > Set CrankRotator(CrankRotator=MakeRotator(Roll=@n16))
   n17 = Get bRigidBodyEnabled()
   Set NoRBAN or FE Alpha(execute=@n1.then_3, NoRBAN or FE Alpha:bool=OR Boolean(A=NOT Boolean(A=@n17), B=Get bIsInFrontEnd())) -> (Output_Get:bool)
   Event OnCharacterCustomizationLoaded()
   > n18 = Sequence!(execute) -> (then_0, then_1)
   > n19 = Cast To FortPlayerPawn(Object=TryGetFortPawn())
   > n20 = WatchCosmeticStat!(
    execute:Exec
    Pawn:Fort Player Pawn Object Reference=@n19
    StatTag:Fort Stat Manager Tag Structure=(TagName="StatManager.AthenaKills")
    DefaultValue:int=0
    bBroadcastInitialValue:bool=false
    InitialValueBroadcastDelay:float=-1
    ChangeValueBroadcastDelay:float=-1
    bConsumeWaitingNotifications:bool=true
   ) -> (then:Exec, OnStatChanged:Exec, StatName:Name, StatValue:int) {is=title K2Node_AsyncAction}
   n21 = Set KillCount(execute=@n20.OnStatChanged, KillCount:double="To Float (Integer)"(InInt=@n20.StatValue)) -> (Output_Get:double)
   > Branch(Condition="float > float"(A=@n21, B=Get Last_KillCount() -> (Last_KillCount:double)) {is=title K2Node_CallFunction})
   > Set Active_Kill(Active_Kill=true)
   > Set Last_KillCount(Last_KillCount:double=@n21) -> (Output_Get:double)
   n22 = Cast To FortPlayerPawn(execute=@n18.then_1, Object=TryGetFortPawn())
   > n23 = WatchCosmeticStat(
    Pawn=@n22
    StatTag=(TagName="StatManager.DamageDealt")
    DefaultValue=0
    bBroadcastInitialValue=false
    InitialValueBroadcastDelay=-1
    ChangeValueBroadcastDelay=-1
    bConsumeWaitingNotifications=true
   )
   n24 = Set DamageCount(execute=@n23.OnStatChanged, DamageCount:double="To Float (Integer)"(InInt=@n23.StatValue)) -> (Output_Get:double)
   > Branch(Condition="float > float"(A=@n24, B=Get Last Damage Count() -> (Last Damage Count:double)))
   > Set Active_Damage(Active_Damage=true)
   > Set Last Damage Count(Last Damage Count:double=@n24) -> (Output_Get:double)
   > Set Crank Count(Crank Count=float + float(A=Get Crank Count(), B=-180))
   TryGetPawnOwner()
  ]}
 SpinFunction {nodes [
   n1 = SpinFunction!() -> (then, Spin:double=~, Speed:double=~) {is=title K2Node_FunctionEntry}
   > n2 = Set spinVar(spinVar:double=@n1) -> (Output_Get:double)
   > n3 = Branch(Condition="float >= float"(A=@n2, B=720) {is=title K2Node_CallFunction})
   > n4 = Set spinVar(spinVar=float - float(A=Get spinVar() -> (spinVar:double), B=720))
   n5 = Get velocity() -> (velocity:double)
   n6 = Get velocityMult() -> (velocityMult:double)
   n7 = Get deltaTimeRemap()
   n8 = Set spinVar(
    execute=@n3.else
    spinVar=float + float!(A=Get spinVar(), B=float * float(A=@n1.Speed, B=@n7), C:double=float * float(A=@n5, B=@n6)) -> (ReturnValue)
   )
   Return Node!(execute=[@n4.then, @n8.then], Output:double=@n8)
  ]}
]
variables [
 Active_Damage {type=boolean}
 Active_Kill {type=boolean}
 Clap Alpha {type=number}
 Clap_Timer {type=number}
 Clap_Timer_Length {type=number, default=100}
 Crank Alpha {type=number}
 Crank Count {type=number}
 Crank_Timer {type=number}
 Crank_Timer_Length {type=number, default=25}
 CrankRotator {type=Rotator}
 crankSpin {type=number}
 DamageCount {type=number}
 deltaTimeRemap {type=number, default=1}
 Head Laugh Rot {type=Rotator}
 Jaw Laugh Rot {type=Rotator}
 KillCount {type=number}
 L Arm Rot {type=Rotator}
 Last Damage Count {type=number}
 Last_KillCount {type=number}
 NoRBAN or FE Alpha {type=boolean}
 R Arm Rot {type=Rotator}
 spinVar {type=number}
 velocity {type=number}
 velocityMult {type=number, default=0.5}
]
